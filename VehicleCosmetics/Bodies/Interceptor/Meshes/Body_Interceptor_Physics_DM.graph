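# Physics asset for the Interceptor vehicle body mesh, named with a DM suffix, providing its simplified collision.
# A single body on the root bone holds one box shape that encloses the whole vehicle, raised and shifted slightly forward from the bone origin.
# It defines no constraints, so the body acts as one rigid piece.
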PhysicsAsset /VehicleCosmetics/Bodies/Interceptor/Meshes/Body_Interceptor_Physics_DM
bodies {root [{shapeType=Box, extentX=444, extentY=216, extentZ=116, shapeName=root_box, center=<31 0 81>}]}
constraints []
